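# Idle head FX for the ThickWatch player body cosmetic, adding glows at the eye and ear positions.
# One emitter bursts four partitioned particles that follow the skeletal mesh, with array lookups per partition setting offset, size, color intensity and alignment for two sprite renderers.
# Waveforms flicker the system main color, EyeBlink modulates alpha on the eye pair, and User.HideEarsGlow fades soft and capsule glows on the ear pair.
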NiagaraSystem /BRCosmetics/Characters/Player/Male/Medium/Bodies/M_MED_ThickWatch/FX/NS_ThickWatch_Head_Idle
dependencies {
 usedDataInterfaces [
  &NiagaraDataInterfaceArrayFloat
  &NiagaraDataInterfaceArrayFloat3
  &NiagaraDataInterfaceSkeletalMesh
 ]
 usedDynamicInputs [
  &SimulationPosition
  &FNTimeOfDayScalar
  &self:ScratchDynamicInput
  &LerpLinearColors
  &FN_FrontEndFloatModifier
 ]
 usedModules [
  &self:SystemSpawnScript.SystemScriptSource.SystemScriptGraph.NiagaraNodeAssignment_0.SetVariables_#1
  &self:SystemSpawnScript.SystemScriptSource.SystemScriptGraph.NiagaraNodeAssignment_3.SetVariables_#2
  &self:SystemSpawnScript.SystemScriptSource.SystemScriptGraph.NiagaraNodeAssignment_1.SetVariables_#3
  &self:SystemSpawnScript.SystemScriptSource.SystemScriptGraph.NiagaraNodeAssignment_2.SetVariables_#4
  &SystemState
 ]
}
emitters {Glow {
  emitterUpdateScript [
   &EmitterState(
    Inactive Response=::NewEnumerator1
    Life Cycle Mode=::NewEnumerator1
    Loop Behavior=::NewEnumerator1
    Loop Duration Mode=::NewEnumerator1
    UseLoopDelay=false
   )
   &SpawnBurst_Instantaneous(Spawn Count=4, Use Spawn Probability=false, UseLoopCountLimit=false)
  ]
  particleSpawnScript [
   &PartitionParticles(Number Of Partitions=4, Particles Per Partition=1, Partition Mode:ENiagara_PartitionMode=::NewEnumerator1)
   &InitializeParticle(Color Mode=::NewEnumerator0, UsePositionOffset=false, Write Material Random=true)
  ]
  particleUpdateScript [
   &ParticleState(Kill Particles When Lifetime Has Elapsed=false, Let Infinitely Lived Particles Die When Emitter Deactivates=true)
   /SetVariables(Particles.LODControl=1) {isSetParameters=true}
   /SetVariables(Particles.MainColor:LinearColor=&MultiplyLinearColor(
     Linear Color=&MultiplyLinearColor(
      Linear Color=$System.MainColor
      Scale Factor=&MakeFloat4FromVectorAndFloat(
       "Vector (RGB)"=<1 1 1>
       "Float (Alpha)"=&Multiply_Float(
        A=&Waveform(Global Phase=$Engine.Time, "Waveform [1]"=::NewEnumerator0, "[1] Phase Scale"=0.3, "[1] Frequency"=1, "[1] Amplitude Min/Max"=<0 1>)
        B=1
       )
      )
     )
     Scale Factor=&MakeVector4_FromFloat(Float=&Waveform(
       Waveform Count=::NewEnumerator1
       Global Input Position=&ReturnNormalizedExecIndex(Normalized Index Scale=9)
       "Waveform [1]"=::NewEnumerator12
       "[1] Frequency"=2.65
       "[1] Amplitude Min/Max"=<0.7 1>
       "Waveform [2] Blend"=::NewEnumerator5
       "[2] Interpolation"=0.5
       "[2] Invert Interpolation"=false
       "Waveform [2]"=::NewEnumerator12
       "[2] Input Position"=0
       "[2] Phase"=$Emitter.Age
       "[2] Phase Scale"=1
       "[2] Compound Frequency"=2
       "[2] Frequency"=16
       "[2] Amplitude Min/Max"=<0.75 1>
       "[2] Scale"=1
       "[2] Bias"=0
       "[2] Exponent"=1
       "[2] Pulse Duration"=0.1
       "[2] Clip Triangle Min/Max"=<0 0>
       "[2] Random Seed"=0
      ))
    )) {enabled=false, isSetParameters=true}
   /SetVariables(Particles.MainColor:LinearColor=&MultiplyLinearColor(
     Linear Color=&MultiplyLinearColor(
      Linear Color=$System.MainColor
      Scale Factor=&MakeFloat4FromVectorAndFloat(
       "Vector (RGB)"=<1 1 1>
       "Float (Alpha)"=&Multiply_Float(
        A=1
        B=&MergeHelper_FloatSwitch(0=$System.EyeBlink, 1=$System.EyeBlink, 2=1, 3=1, Index=$Particles.PartitionParticles.Partition, Num=::NewEnumerator2)
       )
      )
     )
     Scale Factor=&MakeVector4_FromFloat(Float=&Waveform(
       "Waveform [1]"=::NewEnumerator12
       "[1] Frequency"=2.65
       "[1] Amplitude Min/Max"=<0.5 1>
       "Waveform [2] Blend"=::NewEnumerator5
       "[2] Interpolation"=0.5
       "[2] Invert Interpolation"=false
       "Waveform [2]"=::NewEnumerator12
       "[2] Input Position"=0
       "[2] Phase"=$Emitter.Age
       "[2] Phase Scale"=1
       "[2] Frequency"=16
       "[2] Scale"=1
       "[2] Bias"=0
       "[2] Exponent"=1
      ))
    )) {isSetParameters=true}
   /SetVariables(
    Particles.SoftGlowCameraOffset:NiagaraFloat=&SelectFloatFromArray(Direct Array Index=$Particles.PartitionParticles.Partition, Float Selection Array=DataInterface() {propertyValues {FloatData [3,3,1,1]}})
    Particles.SoftGlowColor:LinearColor=&MultiplyLinearColor(
     Linear Color=$StackContext.MainColor
     Scale Factor=&MakeFloat4FromVectorAndFloat(
      "Vector (RGB)"=&VectorFromFloat(Value=&SelectFloatFromArray(Direct Array Index=$Particles.PartitionParticles.Partition, Float Selection Array=DataInterface() {propertyValues {FloatData [5,5,5,5]}}))
      "Float (Alpha)"=&Multiply_Float(
       A=0.2
       B=&Lerp_Float(
        A=1
        B=&MergeHelper_FloatSwitch(0=1, 1=1, 2=0, 3=0, Index=$Particles.PartitionParticles.Partition, Num=::NewEnumerator2)
        Alpha=$User.HideEarsGlow
       )
      )
     )
    )
    Particles.SoftGlowSize:Vector2f=&Vector2DFromFloat(Value=&SelectFloatFromArray(Direct Array Index=$Particles.PartitionParticles.Partition, Float Selection Array=DataInterface() {propertyValues {FloatData [10,10,12,12]}}))
   ) {isSetParameters=true}
   /SetVariables(
    Particles.FakeLightColor:LinearColor=&MultiplyLinearColor(
     Linear Color=$StackContext.MainColor
     Scale Factor=&MakeFloat4FromVectorAndFloat(
      "Vector (RGB)"=&VectorFromFloat(Value=&SelectFloatFromArray(
        Direct Array Index=$Particles.PartitionParticles.Partition
        Float Selection Array=DataInterface() {propertyValues {FloatData [3,3,3,3,2,3,3,3]}}
       ))
      "Float (Alpha)"=1
     )
    )
    Particles.FakeLightSize:Vector2f=&Vector2DFromFloat(Value=&SelectFloatFromArray(
      Direct Array Index=$Particles.PartitionParticles.Partition
      Float Selection Array=DataInterface() {propertyValues {FloatData [17,17,10,10,20,15,15,10]}}
     ))
   ) {isSetParameters=true}
   /SetVariables(
    Particles.CapsuleGlowColor:LinearColor=&MultiplyLinearColor(
     Linear Color=$StackContext.MainColor
     Scale Factor=&MakeFloat4FromVectorAndFloat(
      "Vector (RGB)"=&VectorFromFloat(Value=&SelectFloatFromArray(Direct Array Index=$Particles.PartitionParticles.Partition, Float Selection Array=DataInterface() {propertyValues {FloatData [2,2,1,1]}}))
      "Float (Alpha)"=&Lerp_Float(
       A=1
       B=&MergeHelper_FloatSwitch(0=1, 1=1, 2=0, 3=0, Index=$Particles.PartitionParticles.Partition, Num=::NewEnumerator2)
       Alpha=$User.HideEarsGlow
      )
     )
    )
    Particles.CapsuleSpriteSize:Vector2f=&Vector2DFromFloat(Value=&SelectFloatFromArray(Direct Array Index=$Particles.PartitionParticles.Partition, Float Selection Array=DataInterface() {propertyValues {FloatData [5,5,6,6]}}))
   ) {isSetParameters=true}
   &SkeletalMeshLocation(
    Bone Sampling Mode=::NewEnumerator4
    MaskSpawnGroup=false
    Sampled Position Offset=&SelectVectorFromArray(
     Direct Array Index=$Particles.PartitionParticles.Partition
     Vector Selection Array=DataInterface() {propertyValues {FloatData [<6.0796 11.8893 3.6979>, <6.0796 11.8893 -3.6979>, <12.8064 2.7556 7.9545>, <12.8064 2.7556 -7.9545>]}}
    )
    Skeletal Mesh=$System.SkeletalMesh
   )
   /SetVariables(Particles.CapsuleSpriteAlignment:Vector3f=&RotateVectorByQuaternion(
     Vector To Rotate=&SelectVectorFromArray(
      Direct Array Index=$Particles.PartitionParticles.Partition
      Vector Selection Array=DataInterface() {propertyValues {FloatData [<0.3 -0.5 1>, <-0.3 0.5 1>, <0.6 0 0.5>, <-0.6 0 0.5>]}}
     )
     Quaternion=$Particles.SkeletalMeshLocation.SampledRotation
    )) {isSetParameters=true}
   /SetVariables(Particles.CapsuleSpriteDynamic:Vector4f=&MakeFloat4FromVectorAndFloat(
     "Vector (RGB)"=&NormalizeVector(Vector To Normalize=$Particles.CapsuleSpriteAlignment, Safe Normalize=true, Safe Fallback Threshold=0, Safe Fallback Vector=<0 0 0.001>)
     "Float (Alpha)"=&SelectFloatFromArray(Direct Array Index=$Particles.PartitionParticles.Partition, Float Selection Array=DataInterface() {propertyValues {FloatData [10,10,10,10]}})
    )) {isSetParameters=true}
   /SetVariables(Particles.MeshOrientation=&MultiplyQuaternion(
     Quaternion A=$Particles.SkeletalMeshLocation.SampledRotation
     Quaternion B=&MakeQuatFromAxisAngle(
      Axis=&SelectVectorFromArray(
       Direct Array Index=$Particles.PartitionParticles.Partition
       Vector Selection Array=DataInterface() {propertyValues {FloatData [<0 0 0>, <0 0 0>, <0 0 0>, <0 0 0>, <0 0.1 0>, <-0.8 0 0.05>, <-0.35 -0.08 0>, <0 0 0>]}}
      )
      AngleInDegrees=&SelectFloatFromArray(
       Direct Array Index=$Particles.PartitionParticles.Partition
       Float Selection Array=DataInterface() {propertyValues {FloatData [110,110,90,90,90,150,90,0]}}
      )
     )
    )) {isSetParameters=true}
  ]
  rendererClasses [&NiagaraSpriteRendererProperties]
  renderers [
   {rendererClass=&NiagaraSpriteRendererProperties}
   {rendererClass=&NiagaraSpriteRendererProperties, rendererIndex=1}
  ]
 }}
summary {
 systemName=NS_ThickWatch_Head_Idle
 emitters [{emitterName=Glow, rendererClasses [&NiagaraSpriteRendererProperties]}]
 userVariables [
  User.bIsFrontEnd {type=NiagaraBool}
  User.bIsFrontEndPreview {type=NiagaraBool}
  User.HideEarsGlow {type=NiagaraFloat, defaultValue=0, description=""}
  User.SkeletalMeshLOD {type=NiagaraInt32}
  User.Variant {type=NiagaraFloat, defaultValue=0, description=""}
 ]
}
