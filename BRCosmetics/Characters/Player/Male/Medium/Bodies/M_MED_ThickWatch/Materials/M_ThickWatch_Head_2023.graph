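# Skinned head material for the ThickWatch player character, adding a procedural eye blink.
# It runs the shared character PreFX function through ThickWatch advanced emissive, then overrides Base Color and Emissive before PostFX.
# Two time-based Frac waves, combined with Min and thresholded by BlinkTime, generate irregular blinks.
# Rotated, stretched UVs feed a SphereMask, combined with texture channels and open/close eye parameters to build the blink mask.
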Material /BRCosmetics/Characters/Player/Male/Medium/Bodies/M_MED_ThickWatch/Materials/M_ThickWatch_Head_2023
graph {
 properties {
  bUseMaterialAttributes=true
  bUsedWithMeshDeformer=true
  bUsedWithMorphTargets=true
  bUsedWithSkeletalMesh=true
  pixelDepthOffsetMode=PDOM_Legacy
  shadingModel=MSM_FromMaterialExpression
  usedShadingModels=Default Lit | Subsurface | Clear Coat | Subsurface Profile | Cloth
 }
 nodes [
  n1 = VertexInterpolator(VS=PreSkinnedNormal())
  n2 = MaterialFunctionCall() {materialFunction=&MF_CharacterParent_2023_PreFX}
  n3 = Constant() {r=1}
  n4 = GetMaterialAttributes(None=MaterialFunctionCall!(
    Local Position=@n1.PS
    Alpha=ComponentMask(None=@n2.Tech Art Mask) {b=true}
    Fresnel=Fresnel(ExponentIn=@n3) {baseReflectFraction=0}
    Material=@n2.Result
    Local Normal=@n1.PS
   ) {materialFunction=&/MF_AdvancedEmissive_ThickWatch}.Result) {attributeGetTypes [#1, #2]}
  n5 = NamedRerouteUsage()
  n6 = LinearInterpolate(B=NamedRerouteUsage(), Alpha=NamedRerouteUsage())
  n7 = NamedRerouteUsage()
  n8 = NamedRerouteUsage()
  n9 = LinearInterpolate(A=NamedRerouteUsage(), B=Multiply(A=NamedRerouteUsage(), B=LinearInterpolate(Alpha=@n8) {constA=1, constB=0}), Alpha=NamedRerouteUsage())
  n10 = LinearInterpolate(B=@n9, Alpha=NamedRerouteUsage()) {constA=0.01}
  n11 = NamedRerouteUsage()
  MaterialFunctionCall(
   Material=SetMaterialAttributes!(
    MaterialAttributes=@n4.MaterialAttributes
    Base Color=LinearInterpolate(A=@n5, B=@n6, Alpha=@n7) {constB=0}
    Emissive Color=LinearInterpolate(A=@n9, B=@n10, Alpha=@n11) {constB=0}
   ) {attributeSetTypes [#1, #2]}
   M=@n2.M
   Pre Skinned Local Position=@n2.M
   Distance Blend=@n2.M
   Fresnel=@n2.M
  ) {materialFunction=&MF_CharacterParent_2023_PostFX}
  n12 = MaterialFunctionCall(
   UVs=TextureCoordinate()
   Rotation Center=AppendVector(A=Constant() {r=0.9}, B=Constant() {r=0.35})
   "Rotation Angle (0-1)"=Constant() {r=-0.08}
  ) {materialFunction=&CustomRotator}
  n13 = Constant() {r=50}
  n14 = AppendVector(A=Constant() {r=1}, B=@n13)
  n15 = Constant()
  n16 = Multiply(A=Subtract(A=@n13)) {constB=-0.5}
  n17 = Add(
   A=Add(A=Multiply(A=@n12.Rotated Values, B=@n14), B=AppendVector(A=@n15, B=@n16))
   B=AppendVector(A=Constant() {r=-0.4}, B=Constant() {r=-0.706})
  )
  n18 = Subtract(A=@n13)
  n19 = AppendVector(A=Constant() {r=0.5}, B=Subtract(A=Constant() {r=0.9}, B=Multiply(A=@n18) {constB=0.15}))
  n20 = LinearInterpolate(
   A=ScalarParameter() {defaultValue=3, group=ThickWatch, parameterName=OpenEye}
   B=ScalarParameter() {group=ThickWatch, parameterName=closeEye}
   Alpha=NamedRerouteUsage()
  ) {constB=0.1}
  n21 = NamedRerouteUsage()
  n22 = Constant()
  n23 = If(A=NamedRerouteUsage(), "A > B"=Constant() {r=1}, "A == B"=@n22, "A < B"=@n22) {constB=0.99}
  NamedRerouteDeclaration(None=LinearInterpolate(A=SphereMask(A=@n17, B=@n19, Radius=@n20) {hardnessPercent=70}, B=Multiply(A=@n21) {constB=2}, Alpha=Saturate(None=@n23))) {name=EyeBlinkMask, nodeColor=<rgba 0 0.2031 1 1>}
  n24 = Constant() {r=0.3}
  n25 = Constant() {r=0.356}
  NamedRerouteDeclaration(None=SphereMask(A=TextureCoordinate(), B=AppendVector(A=@n24, B=@n25)) {attenuationRadius=0.13}) {name=The mask around the eyes, nodeColor=<rgba 0 0.1562 1 1>}
  n26 = TextureSampleParameter2D() {
   group=ThickWatch
   parameterName=EyeBlinkTexture
   samplerSource=SSM_Wrap_WorldGroupSettings
   texture=/BRCosmetics/Characters/Player/Male/Medium/Bodies/M_MED_ThickWatch/Textures/T_ThickWatch_TeachArtMask2
  }
  NamedRerouteDeclaration(None=@n26.G) {name=EYE Mask, nodeColor=<rgba 0.875 0 1 1>}
  NamedRerouteDeclaration(None=@n26.B) {name=Closing moment, nodeColor=<rgba 0 1 0.5547 1>}
  NamedRerouteDeclaration(None=@n4.Base Color) {name=BaseColor In, nodeColor=<rgba 1 0.6328 0 1>}
  NamedRerouteDeclaration(None=@n4.Emissive Color) {name=EmissiveColor In, nodeColor=<rgba 0.3125 0 1 1>}
  n27 = Time()
  n28 = Constant() {r=0.3}
  n29 = Constant() {r=0.17}
  n30 = Min(A=Frac(None=Multiply(A=@n27, B=@n28)), B=Frac(None=Multiply(A=@n27, B=@n29) {constB=0.17}))
  n31 = Multiply(A=@n30) {constB=50}
  NamedRerouteDeclaration(None=If(A=@n30, B=ScalarParameter() {defaultValue=0.05, group=ThickWatch, parameterName=BlinkTime}, "A > B"=Constant(), "A < B"=Saturate(None=@n31))) {name=Blinking, nodeColor=<rgba 0 1 0.5547 1>}
 ]
}
