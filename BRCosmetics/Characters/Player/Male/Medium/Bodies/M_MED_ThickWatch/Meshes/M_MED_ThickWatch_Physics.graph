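# Physics asset for the M_MED_ThickWatch male player body cosmetic, covering the pelvis, both thighs and the dangling lace bones with capsule bodies.
# Two asymmetric lace chains, two links on the left and three on the right, hang from the pelvis through tightly limited constraints.
# Both thighs also constrain to the pelvis, with wider swing and twist limits and much larger capsules than the laces.
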PhysicsAsset /BRCosmetics/Characters/Player/Male/Medium/Bodies/M_MED_ThickWatch/Meshes/M_MED_ThickWatch_Physics
bodies {
 dyn_lace_l_01 [{shapeType=Capsule, length=3, radius=1, shapeName=dyn_lace_l_01_capsule, rotation=<rot -90 -106.112 106.1131>}]
 dyn_lace_l_02 [{shapeType=Capsule, length=5, radius=1, shapeName=dyn_lace_l_02_capsule, center=<1.7427 0 0>, rotation=<rot -90 177.9728 -177.9716>}]
 dyn_lace_r_01 [{shapeType=Capsule, length=3, radius=1, shapeName=dyn_lace_r_01_capsule, rotation=<rot -90 -105.4271 105.4282>}]
 dyn_lace_r_02 [{shapeType=Capsule, length=3, radius=1, shapeName=dyn_lace_r_02_capsule, rotation=<rot -90 -92.0361 92.0372>}]
 dyn_lace_r_03 [{shapeType=Capsule, length=5, radius=1, shapeName=dyn_lace_r_03_capsule, center=<1.8323 0 0>, rotation=<rot -90 -92.0367 92.0379>}]
 pelvis [{shapeType=Capsule, length=10, radius=5, center=<0.0765 0.0106 0.1995>, rotation=<rot 0 173.2052 -178.3512>}]
 thigh_l [{shapeType=Capsule, length=17, radius=9, center=<-9.2214 1.1748 -0.6313>, rotation=<rot 0.67 -91.5612 -86.1976>}]
 thigh_r [{shapeType=Capsule, length=17, radius=9, center=<9.6687 -1.04 0.6965>, rotation=<rot 0.8011 -92.2299 -85.4519>}]
}
constraints [
 {bone1Name=thigh_l, bone2Name=pelvis, swing1LimitDegrees=45, swing2LimitDegrees=45, twistLimitDegrees=45}
 {bone1Name=thigh_r, bone2Name=pelvis, swing1LimitDegrees=45, swing2LimitDegrees=45, twistLimitDegrees=45}
 {bone1Name=dyn_lace_l_01, bone2Name=pelvis, swing1LimitDegrees=20, swing2LimitDegrees=20, twistLimitDegrees=20}
 {bone1Name=dyn_lace_l_02, bone2Name=dyn_lace_l_01, swing1LimitDegrees=20, swing2LimitDegrees=20, twistLimitDegrees=20}
 {bone1Name=dyn_lace_r_01, bone2Name=pelvis, swing1LimitDegrees=20, swing2LimitDegrees=20, twistLimitDegrees=20}
 {bone1Name=dyn_lace_r_02, bone2Name=dyn_lace_r_01, swing1LimitDegrees=20, swing2LimitDegrees=20, twistLimitDegrees=20}
 {bone1Name=dyn_lace_r_03, bone2Name=dyn_lace_r_02, swing1LimitDegrees=20, swing2LimitDegrees=20, twistLimitDegrees=20}
]
