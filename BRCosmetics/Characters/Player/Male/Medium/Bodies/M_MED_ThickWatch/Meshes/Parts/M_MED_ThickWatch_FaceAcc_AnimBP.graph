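# Anim Blueprint for the hood bones of the ThickWatch male face accessory, following the parent mesh via Copy Pose From Mesh.
# Additive bone transforms bend or lift the hood when jumping, falling or skydive-floating, then feed a Control Rig.
# The event graph sets the falling flag and interpolates targeting and movement offsets for that rig; "Crouching Move Offset" tracks non-targeting movement.
# An Anim Dynamics hood chain, disabled during melee, is present but disconnected from the output.
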AnimBlueprint /BRCosmetics/Characters/Player/Male/Medium/Bodies/M_MED_ThickWatch/Meshes/Parts/M_MED_ThickWatch_FaceAcc_AnimBP
target_skeleton=&/M_MED_ThickWatch_FaceAcc_Skeleton
graphs [
 AnimGraph {nodes [
   n1 = NOT Boolean(A=Get bIsPlayingMeleeAnim())
   n2 = Anim Dynamics - Bone: dyn_hood_base!(ComponentPose, bAlphaBoolEnabled:bool=@n1, ExternalForce="0,0,0") -> (Pose) {
    alphaInputType=Bool
    angularDampingOverride=1
    bOverrideAngularDamping=true
    bUseGravityOverride=true
    simulationSpace=RootRelative
    alphaBoolBlend {blendInTime=1, blendOption=CircularIn, blendOutTime=1}
    boundBone {boneName=dyn_hood_base}
    componentAppliedLinearAccClamp=<1000 1000 1000>
    componentLinearAccScale=<0.8 0.8 0.3>
    componentLinearVelScale=<0.9 0.9 0.4>
    gravityOverride {z=-980}
    physicsBodyDefinitions [{
      boundBone {boneName=dyn_hood_base}
      boxExtents=<6 20 25>
      constraintSetup {angularLimitsMax=<0 0 0>, angularLimitsMin=<0 0 -35>, angularTarget=<0 0 0>}
      localJointOffset=<1 0 0>
     }]
   }
   n3 = Anim Dynamics - Bone: dyn_hood_01!(ComponentPose=@n2, bAlphaBoolEnabled:bool=@n1, ExternalForce="0,0,0") -> (Pose) {
    alphaInputType=Bool
    angularDampingOverride=1
    bOverrideAngularDamping=true
    bUseGravityOverride=true
    simulationSpace=RootRelative
    alphaBoolBlend {blendInTime=1, blendOption=CircularIn, blendOutTime=1}
    boundBone {boneName=dyn_hood_01}
    componentAppliedLinearAccClamp=<1000 1000 1000>
    componentLinearAccScale=<0.8 0.8 0.3>
    componentLinearVelScale=<0.9 0.9 0.4>
    gravityOverride {z=-980}
    physicsBodyDefinitions [{
      boundBone {boneName=dyn_hood_01}
      boxExtents=<6 20 25>
      constraintSetup {angularLimitsMax=<0 0 0>, angularLimitsMin=<0 0 -35>, angularTarget=<0 0 0>}
      localJointOffset=<1 0 0>
     }]
   }
   n4 = Anim Dynamics - Bone: dyn_hood_02!(ComponentPose=@n3, bAlphaBoolEnabled:bool=@n1, ExternalForce="0,0,0") -> (Pose) {
    alphaInputType=Bool
    angularDampingOverride=1
    bOverrideAngularDamping=true
    bUseGravityOverride=true
    simulationSpace=RootRelative
    alphaBoolBlend {blendInTime=1, blendOption=CircularIn, blendOutTime=1}
    boundBone {boneName=dyn_hood_02}
    componentAppliedLinearAccClamp=<1000 1000 1000>
    componentLinearAccScale=<0.8 0.8 0.3>
    componentLinearVelScale=<0.9 0.9 0.4>
    gravityOverride {z=-980}
    physicsBodyDefinitions [{
      boundBone {boneName=dyn_hood_02}
      boxExtents=<6 20 25>
      constraintSetup {angularLimitsMax=<0 0 0>, angularLimitsMin=<0 0 -35>, angularTarget=<0 0 0>}
      localJointOffset=<1 0 0>
     }]
   }
   Anim Dynamics - Bone: dyn_hood_03!(ComponentPose=@n4, bAlphaBoolEnabled:bool=@n1, ExternalForce="0,0,0") -> (Pose) {
    alphaInputType=Bool
    angularDampingOverride=1
    bOverrideAngularDamping=true
    bUseGravityOverride=true
    simulationSpace=RootRelative
    alphaBoolBlend {blendInTime=1, blendOption=CircularIn, blendOutTime=1}
    boundBone {boneName=dyn_hood_03}
    componentAppliedLinearAccClamp=<1000 1000 1000>
    componentLinearAccScale=<0.8 0.8 0.3>
    componentLinearVelScale=<0.9 0.9 0.4>
    gravityOverride {z=-980}
    physicsBodyDefinitions [{
      boundBone {boneName=dyn_hood_03}
      boxExtents=<6 20 25>
      constraintSetup {angularLimitsMax=<0 0 0>, angularLimitsMin=<0 0 -35>, angularTarget=<0 0 0>}
      localJointOffset=<1 0 0>
     }]
   }
   "Transform (Modify) Bone - Bone: dyn_hood_base"!(ComponentPose, bAlphaBoolEnabled:bool=True, Translation="0,0,0", Rotation="0,-25,0", Scale="1,1,1") -> (Pose) {
    alphaInputType=Bool
    rotationMode=BMM_Additive
    rotationSpace=BCS_BoneSpace
    alphaBoolBlend {blendInTime=0.3, blendOption=Custom, blendOutTime=0.8}
    boneToModify {boneName=dyn_hood_base}
    rotation {yaw=-25}
   }
   n5 = AND Boolean(A=Get bIsJumping(), B=NOT Boolean(A=Get bIsSkyDiving()))
   n6 = "Transform (Modify) Bone - Bone: dyn_hood_01"!(
    ComponentPose=Local To Component(LocalPose=Copy Pose From Mesh() {bUseAttachedParent=true})
    bAlphaBoolEnabled:bool=@n5
    Translation="0,0,0"
    Rotation="0,25,0"
    Scale="1,1,1"
   ) -> (Pose) {
    alphaInputType=Bool
    rotationMode=BMM_Additive
    rotationSpace=BCS_BoneSpace
    alphaBoolBlend {blendInTime=0.2, blendOption=Linear, blendOutTime=0.8}
    boneToModify {boneName=dyn_hood_01}
    rotation {yaw=25}
   }
   n7 = "Transform (Modify) Bone - Bone: dyn_hood_02"!(ComponentPose=@n6, bAlphaBoolEnabled:bool=@n5, Translation="0,0,0", Rotation="0,25,0", Scale="1,1,1") -> (Pose) {
    alphaInputType=Bool
    rotationMode=BMM_Additive
    rotationSpace=BCS_BoneSpace
    alphaBoolBlend {blendInTime=0.2, blendOption=Linear, blendOutTime=0.8}
    boneToModify {boneName=dyn_hood_02}
    rotation {yaw=25}
   }
   n8 = "Transform (Modify) Bone - Bone: dyn_hood_03"!(ComponentPose=@n7, bAlphaBoolEnabled:bool=@n5, Translation="0,0,0", Rotation="0,25,0", Scale="1,1,1") -> (Pose) {
    alphaInputType=Bool
    rotationMode=BMM_Additive
    rotationSpace=BCS_BoneSpace
    alphaBoolBlend {blendInTime=0.2, blendOption=Linear, blendOutTime=0.8}
    boneToModify {boneName=dyn_hood_03}
    rotation {yaw=25}
   }
   n9 = Get HoodJump() -> (HoodJump:bool)
   n10 = "Transform (Modify) Bone - Bone: dyn_hood_base"(ComponentPose=@n8, bAlphaBoolEnabled=@n9, Translation="0,0,0", Rotation="0,-25,0", Scale="1,1,1") {
    alphaInputType=Bool
    rotationMode=BMM_Additive
    rotationSpace=BCS_BoneSpace
    alphaBoolBlend {blendInTime=0.4, blendOption=Custom, blendOutTime=0.2}
    boneToModify {boneName=dyn_hood_base}
    rotation {yaw=-25}
   }
   n11 = "Transform (Modify) Bone - Bone: dyn_hood_01"(ComponentPose=@n10, bAlphaBoolEnabled=@n9, Translation="0,0,0", Rotation="0,-25,0", Scale="1,1,1") {
    alphaInputType=Bool
    rotationMode=BMM_Additive
    rotationSpace=BCS_BoneSpace
    alphaBoolBlend {blendInTime=0.4, blendOption=Linear, blendOutTime=0.2}
    boneToModify {boneName=dyn_hood_01}
    rotation {yaw=-25}
   }
   n12 = "Transform (Modify) Bone - Bone: dyn_hood_02"(ComponentPose=@n11, bAlphaBoolEnabled=@n9, Translation="0,0,0", Rotation="0,-25,0", Scale="1,1,1") {
    alphaInputType=Bool
    rotationMode=BMM_Additive
    rotationSpace=BCS_BoneSpace
    alphaBoolBlend {blendInTime=0.4, blendOption=Linear, blendOutTime=0.2}
    boneToModify {boneName=dyn_hood_02}
    rotation {yaw=-25}
   }
   n13 = "Transform (Modify) Bone - Bone: dyn_hood_03"(ComponentPose=@n12, bAlphaBoolEnabled=@n9, Translation="0,0,0", Rotation="0,-25,0", Scale="1,1,1") {
    alphaInputType=Bool
    rotationMode=BMM_Additive
    rotationSpace=BCS_BoneSpace
    alphaBoolBlend {blendInTime=0.4, blendOption=Linear, blendOutTime=0.2}
    boneToModify {boneName=dyn_hood_03}
    rotation {yaw=-25}
   }
   n14 = "Transform (Modify) Bone - Bone: dyn_hood_03"(ComponentPose=@n13, bAlphaBoolEnabled=Get bIsSkydiveFloating(), Translation="0,0,10", Rotation="0,0,0", Scale="1,1,1") {
    alphaInputType=Bool
    rotationMode=BMM_Additive
    rotationSpace=BCS_BoneSpace
    translationMode=BMM_Additive
    alphaBoolBlend {blendInTime=0.4, blendOption=Linear, blendOutTime=0.7}
    boneToModify {boneName=dyn_hood_03}
    translation {z=10}
   }
   n15 = Control Rig!(
    Source=Component To Local(ComponentPose=@n14)
    Alpha
    targeting_offset:double=Get Targeting Offset() -> (Targeting Offset:double)
    targeting_move_offset:double=Get Targeting Move Offset() -> (Targeting Move Offset:double)
    crouching_move_offset:double=Get Crouching Move Offset() -> (Crouching Move Offset:double)
   ) -> (Pose) {
    alphaBoolBlend {blendOption=Linear}
    controlRigAssetReference {blueprintRigClass=&/M_MED_ThickWatch_FaceAcc_CtrlRig.M_MED_ThickWatch_FaceAcc_CtrlRig_C}
   }
   Output Pose(Result=@n15) {layerGroup=None}
  ]}
 EventGraph {nodes [
   n1 = Event BlueprintUpdateAnimation()
   > Set Delta Time X(Delta Time X=@n1.DeltaTimeX)
   n2 = Get bIsSkyDiving()
   > Set HoodJump(HoodJump:bool=AND Boolean(A=Get bIsFalling(), B=NOT Boolean(A=@n2))) -> (Output_Get:bool)
   n3 = Get bIsTargeting()
   > Set Targeting Offset(Targeting Offset:double=FInterpTo(Current=Get Targeting Offset(), Target="To Float (Boolean)"(InBool=@n3), DeltaTime=Get Delta Time X(), InterpSpeed=10)) -> (Output_Get:double)
   n4 = AND Boolean(A=Get bIsMoving2D(), B=Get bIsTargeting())
   > Set Targeting Move Offset(Targeting Move Offset:double=FInterpTo(Current=Get Targeting Move Offset(), Target="To Float (Boolean)"(InBool=@n4), DeltaTime=Get Delta Time X(), InterpSpeed=10)) -> (Output_Get:double)
   n5 = AND Boolean(A=Get bIsMoving2D(), B=NOT Boolean(A=Get bIsTargeting()))
   > Set Crouching Move Offset(Crouching Move Offset:double=FInterpTo(Current=Get Crouching Move Offset(), Target="To Float (Boolean)"(InBool=@n5), DeltaTime=Get Delta Time X(), InterpSpeed=10)) -> (Output_Get:double)
  ]}
]
variables [
 Crouching Move Offset {type=number}
 Delta Time X {type=number}
 HoodJump {type=boolean}
 Targeting Move Offset {type=number}
 Targeting Offset {type=number}
]
